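# Conversation database for the HenchmanTough NPC, defining its greeting and the top-level branches of its dialogue.
# The entry point leads to one speech node, which has a default line plus contextual lines gated on a first-ever meeting, match conversation count or specific player outfits.
# First-meeting and outfit-reaction lines take priority from named curve rows.
# The speech node then branches through link nodes to shared quest, basic-service and deluxe-service conversations and the NPC's own dialogue entry.
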ConversationDatabase /NPCLibrary/NPCs/HenchmanTough/NPCLibrary_Conversation_HenchmanTough
entry_points [{tag=Conversation.Entry.HenchmanTough, destinations [#1]}]
nodes [
 ConversationEntryPointNode_7 {
  id=#1
  class=ConversationEntryPointNode
  outputs [#2]
  properties {entryTag {tagName=Conversation.Entry.HenchmanTough}}
 }
 Speech_C_4 {
  id=#2
  class=Speech_C
  outputs [#3, #4, #5, #6]
  properties {
   nodeName=""
   generalConfig {
    defaultMessage=You look like you need somethin'.
    contextualMessages [
     {
      message=You look surprised to see me.
      contextRequirements [{requirement {class=FortControllerRequirement_IsFirstEverConversationWithNPC}}]
      priority {curve {rowName=Tandem.Speech.Priority.FirstEverInteraction}}
     }
     {
      message=You look like you need somethin'.
      contextRequirements [{requirement {
         class=FortControllerRequirement_CheckNPCMatchConversationCount
         properties {nPCUniqueIdTag {tagName=AISpawnerData.Type.Specialist.Heavy.HenchmanTough}}
        }}]
      priority {curve {rowName=None}}
     }
     {
      message=You got a message from the Boss?
      contextRequirements [{requirement {
         class=FortControllerRequirement_CheckNPCMatchConversationCount
         properties {nPCUniqueIdTag {tagName=AISpawnerData.Type.Specialist.Heavy.HenchmanTough}}
        }}]
      priority {curve {rowName=None}}
     }
     {
      message="Where Midas goes, I follow."
      contextRequirements [{requirement {
         class=FortControllerRequirement_CheckNPCMatchConversationCount
         properties {nPCUniqueIdTag {tagName=AISpawnerData.Type.Specialist.Heavy.HenchmanTough}}
        }}]
      priority {curve {rowName=None}}
     }
     {
      message=Boss. Good to have ya back.
      contextRequirements [{requirement {class=FortControllerRequirement_HasCID, properties {allowedCIds [&Character_AuraCop_Bullet]}}}]
      priority {curve {rowName=Tandem.Speech.Priority.OutfitReaction}}
     }
     {
      message="Good to see ya, Big Cat."
      contextRequirements [{requirement {
         class=FortControllerRequirement_HasCID
         properties {allowedCIds [&CID_693_Athena_Commando_M_BuffCat]}
        }}]
      priority {curve {rowName=Tandem.Speech.Priority.OutfitReaction}}
     }
    ]
   }
  }
 }
 ConversationLinkNode_3 {id=#6, class=ConversationLinkNode, properties {remoteEntryTag {tagName=Conversation.Entry.DeluxeServices}}}
 ConversationLinkNode_2 {
  id=#5
  class=ConversationLinkNode
  properties {remoteEntryTag {tagName=Conversation.Entry.HenchmanTough.Dialogue}}
 }
 ConversationLinkNode_0 {id=#4, class=ConversationLinkNode, properties {remoteEntryTag {tagName=Conversation.Entry.BasicServices}}}
 ConversationLinkNode_1 {id=#3, class=ConversationLinkNode, properties {remoteEntryTag {tagName=Conversation.Entry.NPCLibrary.Quests}}}
]
